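# Translucent UI material for a login quality-selection tile, showing key art with an animated frame and hover plates.
# A Custom node reads texture dimensions, compared with widget pixel size for aspect-corrected, center-scaled UVs with animated zoom.
# Sharpened UV-distance masks form the border, with an optional top-banner cutout.
# TextHover_Animate slides rotated plate masks and blends their colors, compositing border and plates over the art.
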Material /FortniteCoreUI/UI/Login/QualitySelection/M_QualitySelectionTile
graph {
 properties {
  bUsedWithStaticMesh=false
  blendMode=BLEND_Translucent
  materialDomain=MD_UI
  pixelDepthOffsetMode=PDOM_Legacy
  usedShadingModels=Default Lit
 }
 nodes [
  n1 = VectorParameter() {group=Frame And Shadows, parameterName=FrameBorderColor, defaultValue=<rgba 0 0.0908 1 1>}
  n2 = VectorParameter() {group=Frame And Shadows, parameterName=FrameBorderColor_Hover, defaultValue=<rgba 1 1 1 1>}
  n3 = ScalarParameter() {group=Animators, parameterName=TextHover_Animate, sliderMax=1}
  n4 = VectorParameter() {group=Frame And Shadows, parameterName=YellowHoverColor, defaultValue=<rgba 0.9301 1 0.0103 1>}
  n5 = VectorParameter() {group=Frame And Shadows, parameterName=YellowHoverColor_Hover, defaultValue=<rgba 0.7835 0.9301 1 1>}
  n6 = ScalarParameter() {group=Animators, parameterName=TextHover_Animate, sliderMax=1}
  n7 = TextureCoordinate() {coordinateIndex=4}
  n8 = Rotator(Coordinate=@n7, Time=ScalarParameter() {defaultValue=-0.003, group=Frame And Shadows, parameterName=YellowHover_Rotation}) {speed=6.2832}
  n9 = ScalarParameter() {group=Animators, parameterName=TextHover_Animate, sliderMax=1}
  n10 = Add(A=LinearInterpolate(
    A=ScalarParameter() {defaultValue=-0.125, group=Frame And Shadows, parameterName=YellowHover_PlateStart}
    B=ScalarParameter() {defaultValue=0.05, group=Frame And Shadows, parameterName=YellowHover_PlateFinish}
    Alpha=@n9
   )) {constB=-1}
  n11 = ComponentMask(None=Multiply(A=Add(A=@n8, B=@n10), B=ScalarParameter() {group=Frame And Shadows, parameterName=FrameSharpness})) {g=true}
  n12 = Rotator(
   Coordinate=@n7
   Time=LinearInterpolate(
    A=ScalarParameter() {defaultValue=-0.03, group=Frame And Shadows, parameterName=TextHover_RotationStart}
    B=ScalarParameter() {defaultValue=0.01, group=Frame And Shadows, parameterName=TextHover_RotationFinish}
    Alpha=ScalarParameter() {group=Animators, parameterName=TextHover_Animate, sliderMax=1}
   )
  ) {speed=6.2832}
  n13 = Add(A=LinearInterpolate(
    A=ScalarParameter() {defaultValue=-0.125, group=Frame And Shadows, parameterName=TextHover_PlateStart}
    B=ScalarParameter() {defaultValue=0.275, group=Frame And Shadows, parameterName=TextHover_PlateFinish}
    Alpha=@n9
   )) {constB=-1}
  n14 = ComponentMask(None=Multiply(A=Add(A=@n12, B=@n13), B=ScalarParameter() {group=Frame And Shadows, parameterName=FrameSharpness})) {g=true}
  n15 = TextureCoordinate()
  n16 = Add(
   A=Abs(None=Add(A=@n15) {constB=-0.5})
   B=LinearInterpolate(
    A=ScalarParameter() {defaultValue=0.0312, group=Frame And Shadows, parameterName=FramingEdge}
    B=ScalarParameter() {defaultValue=0.0312, group=Frame And Shadows, parameterName=FramingEdgeHover}
    Alpha=ScalarParameter() {group=Animators, parameterName=TextHover_Animate, sliderMax=1}
   )
  )
  n17 = ScalarParameter() {defaultValue=256, group=Frame And Shadows, parameterName=FrameSharpness}
  n18 = MaterialFunctionCall(Float2=Saturate(None=Multiply(A=@n16, B=@n17))) {materialFunction=&BreakOutFloat2Components}
  n19 = Saturate(None=Add(A=@n18.R, B=@n18.R))
  n20 = ScalarParameter() {defaultValue=-0.1525, group=Frame And Shadows, parameterName=FramingEdgeBanner}
  n21 = Saturate(None=Multiply(A=Add(A=@n15, B=@n20) {constB=-0.1}, B=@n17))
  n22 = LinearInterpolate(A=@n19, Alpha=OneMinus(None=ComponentMask(None=@n21) {g=true}))
  n23 = ScalarParameter() {group=Animators, parameterName=HasTopBanner, sliderMax=1}
  LinearInterpolate(
   A=NamedRerouteUsage()
   B=LinearInterpolate(A=LinearInterpolate(A=@n1.RGB, B=@n2.RGB, Alpha=@n3), B=LinearInterpolate(A=@n4.RGB, B=@n5.RGB, Alpha=@n6), Alpha=Saturate(None=@n11))
   Alpha=LinearInterpolate(A=LinearInterpolate(A=@n19, B=@n22, Alpha=@n23), Alpha=Saturate(None=@n14))
  )
  n24 = TextureCoordinate() {coordinateIndex=4}
  MaterialExpressionTextureObjectParameter_0 = TextureObjectParameter() {group=Image, parameterName=ModeKeyArt, texture=&/MobileQualityModeTileImage_Low}
  n25 = Custom!(Tex=@MaterialExpressionTextureObjectParameter_0) {
   code="float2 dim;\r\nTextureObject.GetDimensions(dim.x, dim.y);\r\nreturn dim;"
   description=GetDimenstions
   outputType=CMOT_Float2
   inputs [{
     inputName=TextureObject
     input {expression=&self:MaterialExpressionTextureObjectParameter_0}
    }]
  }
  n26 = MaterialFunctionCall(Float2=@n25) {materialFunction=&BreakOutFloat2Components}
  n27 = Divide(A=@n26.R, B=@n26.R)
  n28 = MaterialFunctionCall(Float2=MaterialFunctionCall() {materialFunction=&GetUserInterfaceUV}.Pixel Size) {materialFunction=&BreakOutFloat2Components}
  n29 = Divide(A=@n28.R, B=@n28.R)
  n30 = Constant() {r=1}
  n31 = AppendVector(A=Divide(A=@n27, B=@n29), B=@n30)
  n32 = AppendVector(A=@n30, B=Multiply(A=Divide(A=@n26.G, B=@n26.G), B=@n29))
  n33 = Constant() {r=1}
  n34 = Multiply(A=Divide(A=@n28.G, B=@n28.G), B=@n27)
  n35 = Add(
   A=If(
    A=@n29
    B=Constant() {r=1}
    "A > B"=If(A=@n27, B=@n29, "A > B"=@n31, "A < B"=@n32)
    "A == B"=AppendVector(A=@n27, B=@n33)
    "A < B"=AppendVector(A=@n34, B=@n33)
   )
   B=LinearInterpolate(
    A=ScalarParameter() {group=Image, parameterName=ImageScaleAdd_Start}
    B=ScalarParameter() {defaultValue=0.25, group=Image, parameterName=ImageScaleAdd_Finish}
    Alpha=ScalarParameter() {group=Animators, parameterName=ImageScaleAdd_Animate, sliderMax=1}
   )
  )
  NamedRerouteDeclaration(None=TextureSample(
    UVs=MaterialFunctionCall(UVs=@n24, Texture Scale=@n35) {materialFunction=&ScaleUVsByCenter}.UVs
    Tex=@MaterialExpressionTextureObjectParameter_0
   ) {samplerSource=SSM_Clamp_WorldGroupSettings}.RGB) {name=MainImageTexture, nodeColor=<rgba 1 0.2109 0 1>}
  ScalarParameter() {defaultValue=1, parameterName=OverallOpacity}
 ]
}
