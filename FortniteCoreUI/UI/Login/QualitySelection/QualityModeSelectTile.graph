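# Selectable Widget Blueprint tile for one quality level on the login quality-selection screen, showing key art, name, description and download size.
# Setup copies the level's text variables into the text blocks on Construct, at design time and when size info arrives.
# Construct plays the slide-in and feeds the level image into the art's dynamic material; PreConstruct toggles the Recommended banner.
# Selection plays a highlight animation plus confirm flash and shine; deselection reverses the highlight.
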WidgetBlueprint /FortniteCoreUI/UI/Login/QualitySelection/QualityModeSelectTile
animations [
 GameModeSelection {
  binding_count=3
  length_seconds=0.2012
  object=&self:GameModeSelection
  sequence {bindings [
    TileArt {
     id=#1
     possessed_class=Image
     tracks [{
       type=MovieSceneWidgetMaterialTrack
       sections [{
         type=MovieSceneParameterSection
         end_frame=4
         start_frame=0
         channels [TextHover_Animate {type=MovieSceneScriptingFloatChannel, keys [[0,0],[4.0237,1]]}]
        }]
      }]
    }
    Tag_Selected {
     id=#2
     possessed_class=Border
     tracks [
      {
       type=MovieSceneByteTrack
       display_name=Visibility
       sections [{type=MovieSceneByteSection, end_frame=4, start_frame=0, channels [None {type=MovieSceneScriptingByteChannel, keys [[0,1],[0.4143,3]]}]}]
      }
      {
       type=MovieScene2DTransformTrack
       sections [{
         type=MovieScene2DTransformSection
         end_frame=4
         start_frame=0
         channels [
          Translation.X {type=MovieSceneScriptingFloatChannel, const [0,1]}
          Translation.Y {type=MovieSceneScriptingFloatChannel, const [0,1]}
          Angle {type=MovieSceneScriptingFloatChannel, keys [[0,0],[2,-6],[2.485,0]]}
          Scale.X {type=MovieSceneScriptingFloatChannel, keys [[0,1],[2,2],[2.911,1],[4,1]]}
          Scale.Y {type=MovieSceneScriptingFloatChannel, keys [[0,1],[2,2.6922],[2.911,1],[4,1]]}
          Shear.X {type=MovieSceneScriptingFloatChannel, const [0,1]}
          Shear.Y {type=MovieSceneScriptingFloatChannel, const [0,1]}
         ]
        }]
      }
      {
       type=MovieSceneFloatTrack
       display_name=Render Opacity
       sections [{type=MovieSceneFloatSection, end_frame=4, start_frame=0, channels [None {type=MovieSceneScriptingFloatChannel, keys [[0,0],[4,1]]}]}]
      }
     ]
    }
    Overlay_3 {
     id=#3
     possessed_class=Overlay
     tracks [
      {
       type=MovieScene2DTransformTrack
       sections [{
         type=MovieScene2DTransformSection
         end_frame=4
         start_frame=0
         channels [
          Translation.X {type=MovieSceneScriptingFloatChannel}
          Translation.Y {type=MovieSceneScriptingFloatChannel}
          Angle {type=MovieSceneScriptingFloatChannel}
          Scale.X {type=MovieSceneScriptingFloatChannel, keys [[0,1],[4,1.05],[4.0237,1.07]]}
          Scale.Y {type=MovieSceneScriptingFloatChannel, keys [[0,1],[4,1.05],[4.0237,1.07]]}
          Shear.X {type=MovieSceneScriptingFloatChannel}
          Shear.Y {type=MovieSceneScriptingFloatChannel}
         ]
        }]
      }
      {
       type=MovieSceneFloatTrack
       display_name=Render Opacity
       sections [{type=MovieSceneFloatSection, end_frame=4, start_frame=0, channels [None {type=MovieSceneScriptingFloatChannel, const [1,1]}]}]
      }
     ]
    }
   ]}
 }
 ConfirmSeleccion {
  binding_count=2
  length_seconds=0.25
  object=&self:ConfirmSeleccion
  sequence {bindings [
    ConfirmSelectionFlash {
     id=#4
     possessed_class=Image
     tracks [
      {
       type=MovieSceneFloatTrack
       display_name=Render Opacity
       sections [{
         type=MovieSceneFloatSection
         end_frame=5
         start_frame=0
         channels [None {type=MovieSceneScriptingFloatChannel, keys [[0,1],[1,0.75],[3,0.75],[4,0]]}]
        }]
      }
      {
       type=MovieSceneByteTrack
       display_name=Visibility
       sections [{type=MovieSceneByteSection, end_frame=5, start_frame=0, channels [None {type=MovieSceneScriptingByteChannel, keys [[0,3],[5,1]]}]}]
      }
     ]
    }
    ConfirmSelectionShine {
     id=#5
     possessed_class=Image
     tracks [
      {
       type=MovieSceneWidgetMaterialTrack
       sections [{
         type=MovieSceneParameterSection
         end_frame=5
         start_frame=0
         channels [ShineAnim {type=MovieSceneScriptingFloatChannel, keys [[1,0.05],[5,1]]}]
        }]
      }
      {
       type=MovieSceneByteTrack
       display_name=Visibility
       sections [{type=MovieSceneByteSection, end_frame=5, start_frame=0, channels [None {type=MovieSceneScriptingByteChannel, keys [[0,4],[5,1]]}]}]
      }
     ]
    }
   ]}
 }
 InAnOutro {
  binding_count=1
  length_seconds=0.15
  object=&self:InAnOutro
  sequence {bindings [Tile {
     id=#6
     possessed_class=Overlay
     tracks [
      {
       type=MovieScene2DTransformTrack
       sections [{
         type=MovieScene2DTransformSection
         end_frame=3
         start_frame=0
         channels [
          Translation.X {type=MovieSceneScriptingFloatChannel, keys [[0,-494.0451],[3,0]]}
          Translation.Y {type=MovieSceneScriptingFloatChannel, keys [[0,85.2215],[3,0]]}
          Angle {type=MovieSceneScriptingFloatChannel, const [0,1]}
          Scale.X {type=MovieSceneScriptingFloatChannel, const [1,1]}
          Scale.Y {type=MovieSceneScriptingFloatChannel, const [1,1]}
          Shear.X {type=MovieSceneScriptingFloatChannel, const [0,1]}
          Shear.Y {type=MovieSceneScriptingFloatChannel, const [0,1]}
         ]
        }]
      }
      {
       type=MovieSceneFloatTrack
       display_name=Render Opacity
       sections [{type=MovieSceneFloatSection, end_frame=3, start_frame=0, channels [None {type=MovieSceneScriptingFloatChannel, keys [[0,0],[3,1]]}]}]
      }
     ]
    }]}
 }
]
graphs [
 EventGraph {nodes [
   Event Construct()
   > Setup()
   > Play Animation!(
    execute
    self:User Widget Object Reference
    InAnimation:Widget Animation Object Reference=Get InAnOutro() -> (InAnOutro:Widget Animation Object Reference)
    StartAtTime:float=0
    NumLoopsToPlay:int=1
    PlayMode:EUMGSequencePlayMode Enum
    PlaybackSpeed:float=1
    bRestoreState:bool=false
   ) -> (then, ReturnValue:Widget Animation Handle Structure)
   > n1 = GetDynamicMaterial(self=Get TileArt() -> (TileArt:Image Object Reference))
   > SetTextureParameterValue(self=@n1, ParameterName=ModeKeyArt, Value=Get QualityLevelImage() -> (QualityLevelImage:Texture 2D Object Reference))
   n2 = Event PreConstruct()
   > n3 = Branch(Condition=@n2.IsDesignTime)
   > n4 = Setup()
   Set Visibility!(
    execute=[@n4.then, @n3.else]
    self:Widget Object Reference=Get Border_Recommended() -> (Border_Recommended:Common Border Object Reference)
    InVisibility:ESlateVisibility Enum=Select(Option 0=Hidden, Option 1=SelfHitTestInvisible, Index=Get bIsRecommendedLevel() -> (bIsRecommendedLevel:bool))
   ) -> (then) {is=title K2Node_CallFunction}
   n5 = Event OnSelectedStateChanged!() -> (OutputDelegate, then, bIsSelected:bool) {is=title K2Node_Event}
   > n6 = Branch(Condition=@n5.bIsSelected)
   n7 = Get UnhoveredVignette() -> (UnhoveredVignette:Image Object Reference)
   > Set Visibility(self=@n7, InVisibility=Collapsed)
   n8 = Get GameModeSelection() -> (GameModeSelection:Widget Animation Object Reference)
   > Play Animation(InAnimation=@n8, StartAtTime=0, NumLoopsToPlay=1, PlaybackSpeed=1, bRestoreState=false)
   n9 = Get SelectedIcon() -> (SelectedIcon:Image Object Reference)
   > Set Visibility(self=@n9, InVisibility=Collapsed)
   > Play Animation(
    InAnimation=Get ConfirmSeleccion() -> (ConfirmSeleccion:Widget Animation Object Reference)
    StartAtTime=0
    NumLoopsToPlay=1
    PlaybackSpeed=1
    bRestoreState=false
   )
   Set Visibility(execute=@n6.else, self=@n7, InVisibility=Collapsed)
   > Play Animation(InAnimation=@n8, StartAtTime=0, NumLoopsToPlay=1, PlayMode=Reverse, PlaybackSpeed=1, bRestoreState=false)
   > Set Visibility(self=@n9, InVisibility=Collapsed)
   Event OnSizeInfoReceived() {is=title K2Node_Event}
   > Setup()
   Set Visibility(InVisibility=Collapsed)
  ]}
 Setup {nodes [
   Setup() {is=title K2Node_FunctionEntry}
   > n1 = Sequence!(execute) -> (then_0, then_1, then_2:Exec, then_3:Exec, then_4:Exec)
   > "SetText (Text)"(
    self=Get TextBlock_QualityLevelName() -> (TextBlock_QualityLevelName:Common Text Object Reference)
    InText=Get QualityLevelName() -> (QualityLevelName:Text)
   )
   "SetText (Text)"(
    execute=@n1.then_1
    self=Get TextBlock_QualityDescription() -> (TextBlock_QualityDescription:Common Text Object Reference)
    InText=Get QualityLevelDescription() -> (QualityLevelDescription:Text)
   )
   "SetText (Text)"(
    execute=@n1.then_2
    self=Get TextBlock_AdditionalDownloadSize() -> (TextBlock_AdditionalDownloadSize:Common Text Object Reference)
    InText=Get AdditionalDownloadSize() -> (AdditionalDownloadSize:Text)
   )
   "SetText (Text)"(
    execute=@n1.then_3
    self=Get TextBlock_TotalBuildSize() -> (TextBlock_TotalBuildSize:Common Text Object Reference)
    InText=Get TotalBuildSize() -> (TotalBuildSize:Text)
   )
  ]}
]
variables [bIsRecommendedLevel {type=boolean, default=~}]
widgets {
 info {inheritedWidgetCount=0, parentClass=null, rootWidgetClass=&Overlay, widgetCount=31}
 widgets [
  Tile {widgetClassPath=&Overlay, widget=Tile}
  Overlay_3 {
   widgetClassPath=&Overlay
   parent=Tile
   widget=Overlay_3
   properties {clipping=ClipToBounds}
   slot {class=OverlaySlot, properties {horizontalAlignment=HAlign_Fill, verticalAlignment=VAlign_Fill}}
  }
  UnhoveredVignette {
   widgetClassPath=&Image
   bIsVariable=true
   parent=Overlay_3
   widget=UnhoveredVignette
   properties {
    visibility=SelfHitTestInvisible
    brush {
     bIsDynamicallyLoaded=false
     drawAs=Box
     imageType=NoImage
     mirroring=NoMirror
     resourceName=None
     resourceObject=/FortUILibrary/Styles/Border/Materials/M_Shadow_Vignette_BoxScaleSoft
     tiling=NoTile
     imageSize=<256 256>
     margin=<ltrb 0.5 0.5 0.5 0.5>
     outlineSettings {
      bUseBrushTransparency=false
      roundingType=HalfHeightRadius
      width=0
      color {colorUseRule=UseColor_Specified, specifiedColor=<rgba 0 0 0 0>}
      cornerRadii=<0 0 0 0>
     }
     tintColor {colorUseRule=UseColor_Specified, specifiedColor=<rgba 1 1 1 1>}
     uVRegion {bIsValid=false, max=<0 0>, min=<0 0>}
    }
    colorAndOpacity=<rgba 1 1 1 0.7>
   }
   slot {class=OverlaySlot, properties {horizontalAlignment=HAlign_Fill, verticalAlignment=VAlign_Fill}}
  }
  TileArt {
   widgetClassPath=&Image
   bIsVariable=true
   parent=Overlay_3
   widget=TileArt
   properties {brush {
     bIsDynamicallyLoaded=false
     drawAs=Image
     imageType=NoImage
     mirroring=NoMirror
     resourceName=None
     resourceObject=&/MI_QualitySelectionTile_tile
     tiling=NoTile
     imageSize=<32 32>
     margin=<ltrb 0 0 0 0>
     outlineSettings {
      bUseBrushTransparency=false
      roundingType=HalfHeightRadius
      width=0
      color {colorUseRule=UseColor_Specified, specifiedColor=<rgba 0 0 0 0>}
      cornerRadii=<0 0 0 0>
     }
     tintColor {colorUseRule=UseColor_Specified, specifiedColor=<rgba 1 1 1 1>}
     uVRegion {bIsValid=false, max=<0 0>, min=<0 0>}
    }}
   slot {class=OverlaySlot, properties {horizontalAlignment=HAlign_Fill, verticalAlignment=VAlign_Fill}}
  }
  DisabledGray {
   widgetClassPath=&Image
   bIsVariable=true
   parent=Overlay_3
   widget=DisabledGray
   properties {
    visibility=Collapsed
    brush {
     bIsDynamicallyLoaded=false
     drawAs=Image
     imageType=NoImage
     mirroring=NoMirror
     resourceName=None
     resourceObject=null
     tiling=NoTile
     imageSize=<32 32>
     margin=<ltrb 0 0 0 0>
     outlineSettings {
      bUseBrushTransparency=false
      roundingType=HalfHeightRadius
      width=0
      color {colorUseRule=UseColor_Specified, specifiedColor=<rgba 0 0 0 0>}
      cornerRadii=<0 0 0 0>
     }
     tintColor {colorUseRule=UseColor_Specified, specifiedColor=<rgba 0.0561 0.0648 0.1119 1>}
     uVRegion {bIsValid=false, max=<0 0>, min=<0 0>}
    }
    colorAndOpacity=<rgba 1 1 1 0.66>
   }
   slot {class=OverlaySlot, properties {horizontalAlignment=HAlign_Fill, verticalAlignment=VAlign_Fill}}
  }
  ConfirmSelectionShine {
   widgetClassPath=&Image
   bIsVariable=true
   parent=Overlay_3
   widget=ConfirmSelectionShine
   properties {
    visibility=Collapsed
    brush {
     bIsDynamicallyLoaded=false
     drawAs=Image
     imageType=NoImage
     mirroring=NoMirror
     resourceName=None
     resourceObject=/FortniteCoreUI/UI/Foundation/Materials/M_Athena_Matchmaking_ShineAnim_Tile
     tiling=NoTile
     imageSize=<32 32>
     margin=<ltrb 0 0 0 0>
     outlineSettings {
      bUseBrushTransparency=false
      roundingType=HalfHeightRadius
      width=0
      color {colorUseRule=UseColor_Specified, specifiedColor=<rgba 0 0 0 0>}
      cornerRadii=<0 0 0 0>
     }
     tintColor {colorUseRule=UseColor_Specified, specifiedColor=<rgba 1 1 1 1>}
     uVRegion {bIsValid=false, max=<0 0>, min=<0 0>}
    }
   }
   slot {class=OverlaySlot, properties {horizontalAlignment=HAlign_Fill, verticalAlignment=VAlign_Fill}}
  }
  ConfirmSelectionFlash {
   widgetClassPath=&Image
   bIsVariable=true
   parent=Overlay_3
   widget=ConfirmSelectionFlash
   properties {visibility=Collapsed}
   slot {class=OverlaySlot, properties {horizontalAlignment=HAlign_Fill, verticalAlignment=VAlign_Fill}}
  }
  Border_Recommended {
   widgetClassPath=&CommonBorder
   bIsVariable=true
   parent=Overlay_3
   widget=Border_Recommended
   properties {
    visibility=Collapsed
    background {
     bIsDynamicallyLoaded=false
     drawAs=Box
     imageType=NoImage
     mirroring=NoMirror
     resourceName=None
     resourceObject=/Game/UI/Login/SubGameSelect/Materials/M_Notification_Right
     tiling=NoTile
     imageSize=<96 32>
     margin=<ltrb 0.1 0 0.25 0>
     outlineSettings {
      bUseBrushTransparency=false
      roundingType=HalfHeightRadius
      width=0
      color {colorUseRule=UseColor_Specified, specifiedColor=<rgba 0 0 0 0>}
      cornerRadii=<0 0 0 0>
     }
     tintColor {colorUseRule=UseColor_Specified, specifiedColor=<rgba 1 1 1 1>}
     uVRegion {bIsValid=false, max=<0 0>, min=<0 0>}
    }
    padding=<ltrb 20 6 28 6>
    renderTransform {angle=0, scale=<1 1>, shear=<0 0>, translation=<0 -17>}
   }
   slot {class=OverlaySlot}
  }
  SizeBox_0 {
   widgetClassPath=&SizeBox
   parent=Border_Recommended
   widget=SizeBox_0
   properties {bOverride_MaxDesiredWidth=true, maxDesiredWidth=256}
   slot {class=BorderSlot, properties {padding=<ltrb 20 6 28 6>}}
  }
  T_Recommended {
   widgetClassPath=&CommonTextBlock
   bIsVariable=true
   parent=SizeBox_0
   widget=T_Recommended
   properties {
    scrollStyle=&TextScrollStyle-Base_C
    style=/FortUILibrary/Styles/Typography/Foundation/TextStyle-Header-S.TextStyle-Header-S_C
    text=RECOMMENDED
    font {
     bForceMonospaced=false
     bMaterialIsStencil=false
     fontMaterial=null
     fontObject=&Burbank
     letterSpacing=0
     monospacedWidth=1
     size=24
     skewAmount=0
     typefaceFontName=Burbank-Black
     outlineSettings {
      bApplyOutlineToDropShadows=false
      bMiteredCorners=false
      bSeparateFillAlpha=false
      outlineMaterial=null
      outlineSize=0
      outlineColor=<rgba 0 0 0 1>
     }
    }
    shadowOffset=<0 0>
   }
   slot {class=SizeBoxSlot}
  }
  DetailsSB {
   widgetClassPath=&Overlay
   parent=Overlay_3
   widget=DetailsSB
   slot {class=OverlaySlot, properties {horizontalAlignment=HAlign_Fill, verticalAlignment=VAlign_Fill}}
  }
  VerticalBox_0 {
   widgetClassPath=&VerticalBox
   parent=DetailsSB
   widget=VerticalBox_0
   slot {class=OverlaySlot, properties {horizontalAlignment=HAlign_Fill, verticalAlignment=VAlign_Fill}}
  }
  Spacer_199 {
   widgetClassPath=&Spacer
   parent=VerticalBox_0
   widget=Spacer_199
   slot {class=VerticalBoxSlot, properties {size {sizeRule=Fill, value=0.85}}}
  }
  FooterArrowBg {
   widgetClassPath=&Overlay
   parent=VerticalBox_0
   widget=FooterArrowBg
   slot {class=VerticalBoxSlot, properties {size {sizeRule=Fill, value=0.2}}}
  }
  NameOvr {
   widgetClassPath=&Overlay
   parent=FooterArrowBg
   widget=NameOvr
   slot {class=OverlaySlot, properties {horizontalAlignment=HAlign_Fill, verticalAlignment=VAlign_Fill, padding=<ltrb 20 0 0 0>}}
  }
  NameHB {
   widgetClassPath=&HorizontalBox
   parent=NameOvr
   widget=NameHB
   slot {class=OverlaySlot, properties {horizontalAlignment=HAlign_Fill, verticalAlignment=VAlign_Fill}}
  }
  ScaleBox_1 {
   widgetClassPath=&ScaleBox
   parent=NameHB
   widget=ScaleBox_1
   slot {class=HorizontalBoxSlot, properties {padding=<ltrb 0 20 0 20>, size {sizeRule=Fill, value=0.2}}}
  }
  TextBlock_QualityLevelName {
   widgetClassPath=&CommonTextBlock
   bIsVariable=true
   parent=ScaleBox_1
   widget=TextBlock_QualityLevelName
   properties {
    text=LOW
    colorAndOpacity {colorUseRule=UseColor_Specified, specifiedColor=<rgba 0 0 0 1>}
    font {
     bForceMonospaced=false
     bMaterialIsStencil=false
     fontMaterial=null
     fontObject=&Burbank
     letterSpacing=0
     monospacedWidth=1
     size=40
     skewAmount=0
     typefaceFontName=Burbank-Black
     outlineSettings {
      bApplyOutlineToDropShadows=false
      bMiteredCorners=false
      bSeparateFillAlpha=false
      outlineMaterial=null
      outlineSize=0
      outlineColor=<rgba 0 0 0 1>
     }
    }
    renderTransform {angle=0, scale=<1 1>, shear=<-2 0>, translation=<0 1>}
    shadowOffset=<0 0>
   }
   slot {class=ScaleBoxSlot, properties {horizontalAlignment=HAlign_Fill}}
  }
  Description {
   widgetClassPath=&Overlay
   parent=NameHB
   widget=Description
   slot {class=HorizontalBoxSlot, properties {verticalAlignment=VAlign_Center, padding=<ltrb 0 0 0 5>, size {sizeRule=Fill, value=0.8}}}
  }
  TextBlock_QualityDescription {
   widgetClassPath=&CommonTextBlock
   bIsVariable=true
   parent=Description
   widget=TextBlock_QualityDescription
   properties {
    autoWrapText=true
    justification=Center
    text=High Quality Level High Quality Level High Quality Level
    colorAndOpacity {colorUseRule=UseColor_Specified, specifiedColor=<rgba 0 0 0 1>}
    font {
     bForceMonospaced=false
     bMaterialIsStencil=false
     fontMaterial=null
     fontObject=&BurbankSmall
     letterSpacing=0
     monospacedWidth=1
     size=21
     skewAmount=0
     typefaceFontName=Bold
     outlineSettings {
      bApplyOutlineToDropShadows=false
      bMiteredCorners=false
      bSeparateFillAlpha=false
      outlineMaterial=null
      outlineSize=0
      outlineColor=<rgba 0 0 0 1>
     }
    }
    renderTransform {angle=0, scale=<1 1>, shear=<-8 0>, translation=<0 0>}
    shadowOffset=<0 0>
   }
   slot {class=OverlaySlot, properties {horizontalAlignment=HAlign_Fill, verticalAlignment=VAlign_Fill, padding=<ltrb 20 0 20 0>}}
  }
  SelectedIcon {
   widgetClassPath=&Image
   bIsVariable=true
   parent=NameOvr
   widget=SelectedIcon
   properties {
    visibility=Hidden
    brush {
     bIsDynamicallyLoaded=false
     drawAs=Image
     imageType=NoImage
     mirroring=NoMirror
     resourceName=None
     resourceObject=/Game/UI/Foundation/Textures/Icons/Checks/T-Icon-Check-64
     tiling=NoTile
     imageSize=<34 34>
     margin=<ltrb 0 0 0 0>
     outlineSettings {
      bUseBrushTransparency=false
      roundingType=HalfHeightRadius
      width=0
      color {colorUseRule=UseColor_Specified, specifiedColor=<rgba 0 0 0 0>}
      cornerRadii=<0 0 0 0>
     }
     tintColor {colorUseRule=UseColor_Specified, specifiedColor=<rgba 0 1 0 1>}
     uVRegion {bIsValid=false, max=<0 0>, min=<0 0>}
    }
    colorAndOpacity=<rgba 1 1 1 0.7>
   }
   slot {class=OverlaySlot, properties {padding=<ltrb 0 4 0 0>}}
  }
  HorizontalBox_1 {
   widgetClassPath=&HorizontalBox
   parent=FooterArrowBg
   widget=HorizontalBox_1
   slot {class=OverlaySlot, properties {horizontalAlignment=HAlign_Right, verticalAlignment=VAlign_Bottom, padding=<ltrb 0 0 6 4>}}
  }
  PlusIcon {
   widgetClassPath=&CommonTextBlock
   parent=HorizontalBox_1
   widget=PlusIcon
   properties {
    justification=Right
    text=+
    colorAndOpacity {colorUseRule=UseColor_Specified, specifiedColor=<rgba 0 0 0 1>}
    font {
     bForceMonospaced=false
     bMaterialIsStencil=false
     fontMaterial=null
     fontObject=&BurbankSmall
     letterSpacing=0
     monospacedWidth=1
     size=24
     skewAmount=0
     typefaceFontName=Bold
     outlineSettings {
      bApplyOutlineToDropShadows=false
      bMiteredCorners=false
      bSeparateFillAlpha=false
      outlineMaterial=null
      outlineSize=2
      outlineColor=<rgba 0.0704 0.9647 1 1>
     }
    }
    renderTransform {angle=0, scale=<1 1>, shear=<-6 0>, translation=<0 8>}
    shadowOffset=<0 0>
   }
   slot {class=HorizontalBoxSlot, properties {horizontalAlignment=HAlign_Right, verticalAlignment=VAlign_Bottom}}
  }
  TextBlock_AdditionalDownloadSize {
   widgetClassPath=&CommonTextBlock
   bIsVariable=true
   parent=HorizontalBox_1
   widget=TextBlock_AdditionalDownloadSize
   properties {
    justification=Right
    text=5Gb
    colorAndOpacity {colorUseRule=UseColor_Specified, specifiedColor=<rgba 0 0 0 1>}
    font {
     bForceMonospaced=false
     bMaterialIsStencil=false
     fontMaterial=null
     fontObject=&BurbankSmall
     letterSpacing=0
     monospacedWidth=1
     size=23
     skewAmount=0
     typefaceFontName=Black
     outlineSettings {
      bApplyOutlineToDropShadows=false
      bMiteredCorners=false
      bSeparateFillAlpha=false
      outlineMaterial=null
      outlineSize=2
      outlineColor=<rgba 0.0704 0.9647 1 1>
     }
    }
    renderTransform {angle=0, scale=<1 1>, shear=<-8 0>, translation=<0 4>}
    shadowOffset=<0 0>
   }
   slot {class=HorizontalBoxSlot, properties {horizontalAlignment=HAlign_Right, verticalAlignment=VAlign_Bottom}}
  }
  Header_Info {
   widgetClassPath=&HorizontalBox
   parent=DetailsSB
   widget=Header_Info
   slot {class=OverlaySlot, properties {horizontalAlignment=HAlign_Center, padding=<ltrb 0 10 0 0>}}
  }
  CommonTextBlock_60 {
   widgetClassPath=&CommonTextBlock
   parent=Header_Info
   widget=CommonTextBlock_60
   properties {
    justification=Center
    scrollStyle=&TextScrollStyle-Base_C
    text=Total game size:
    visibility=Collapsed
    colorAndOpacity {colorUseRule=UseColor_Specified, specifiedColor=<rgba 0.0885 0.0885 0.0885 1>}
    font {
     bForceMonospaced=false
     bMaterialIsStencil=false
     fontMaterial=null
     fontObject=&BurbankSmall
     letterSpacing=0
     monospacedWidth=1
     size=15
     skewAmount=0
     typefaceFontName=Bold
     outlineSettings {
      bApplyOutlineToDropShadows=false
      bMiteredCorners=false
      bSeparateFillAlpha=false
      outlineMaterial=null
      outlineSize=0
      outlineColor=<rgba 0 0 0 1>
     }
    }
    renderTransform {angle=0, scale=<1 1>, shear=<-12 0>, translation=<0 0>}
    shadowOffset=<0 0>
   }
   slot {class=HorizontalBoxSlot, properties {horizontalAlignment=HAlign_Right, size {sizeRule=Fill, value=1}}}
  }
  TextBlock_TotalBuildSize {
   widgetClassPath=&CommonTextBlock
   bIsVariable=true
   parent=Header_Info
   widget=TextBlock_TotalBuildSize
   properties {
    justification=Center
    text=X.X GB
    visibility=Collapsed
    colorAndOpacity {colorUseRule=UseColor_Specified, specifiedColor=<rgba 0 0 0 1>}
    font {
     bForceMonospaced=false
     bMaterialIsStencil=false
     fontMaterial=null
     fontObject=&BurbankSmall
     letterSpacing=0
     monospacedWidth=1
     size=15
     skewAmount=0
     typefaceFontName=Bold
     outlineSettings {
      bApplyOutlineToDropShadows=false
      bMiteredCorners=false
      bSeparateFillAlpha=false
      outlineMaterial=null
      outlineSize=0
      outlineColor=<rgba 0 0 0 1>
     }
    }
    renderTransform {angle=0, scale=<1 1>, shear=<-12 0>, translation=<0 0>}
    shadowOffset=<0 0>
   }
   slot {class=HorizontalBoxSlot, properties {horizontalAlignment=HAlign_Left, padding=<ltrb 7 0 0 0>}}
  }
  Tag_Selected {
   widgetClassPath=&Border
   bIsVariable=true
   parent=Tile
   widget=Tag_Selected
   properties {
    visibility=Collapsed
    background {
     bIsDynamicallyLoaded=false
     drawAs=Box
     imageType=NoImage
     mirroring=NoMirror
     resourceName=None
     resourceObject=/Game/Athena/UI/ReplayBrowser/M_Ui_Skewed_NoV1
     tiling=NoTile
     imageSize=<32 32>
     margin=<ltrb 0.25 0 0.25 0>
     outlineSettings {
      bUseBrushTransparency=false
      roundingType=HalfHeightRadius
      width=0
      color {colorUseRule=UseColor_Specified, specifiedColor=<rgba 0 0 0 0>}
      cornerRadii=<0 0 0 0>
     }
     tintColor {colorUseRule=UseColor_Specified, specifiedColor=<rgba 0.0343 1 0.227 1>}
     uVRegion {bIsValid=false, max=<0 0>, min=<0 0>}
    }
    contentColorAndOpacity=<rgba 0 0.147 0.063 1>
    padding=<ltrb 10 2 10 2>
   }
   slot {class=OverlaySlot, properties {horizontalAlignment=HAlign_Right, padding=<ltrb 20 20 20 20>}}
  }
  HorizontalBox_5 {
   widgetClassPath=&HorizontalBox
   parent=Tag_Selected
   widget=HorizontalBox_5
   slot {class=BorderSlot, properties {padding=<ltrb 10 2 10 2>}}
  }
  Image_140 {
   widgetClassPath=&Image
   bIsVariable=true
   parent=HorizontalBox_5
   widget=Image_140
   properties {
    brush {
     bIsDynamicallyLoaded=false
     drawAs=Image
     imageType=NoImage
     mirroring=NoMirror
     resourceName=None
     resourceObject=/Game/Athena/HUD/Bacchus/HUDLayoutTool/HLT_Icons/T-HLT-CheckMark-32
     tiling=NoTile
     imageSize=<16 16>
     margin=<ltrb 0 0 0 0>
     outlineSettings {
      bUseBrushTransparency=false
      roundingType=HalfHeightRadius
      width=0
      color {colorUseRule=UseColor_Specified, specifiedColor=<rgba 0 0 0 0>}
      cornerRadii=<0 0 0 0>
     }
     tintColor {colorUseRule=UseColor_Specified, specifiedColor=<rgba 1 1 1 1>}
     uVRegion {bIsValid=false, max=<0 0>, min=<0 0>}
    }
    renderTransform {angle=0, scale=<1.3 1.3>, shear=<0 0>, translation=<0 0>}
   }
   slot {class=HorizontalBoxSlot, properties {horizontalAlignment=HAlign_Left, verticalAlignment=VAlign_Center, padding=<ltrb 0 -1.5 6 0>}}
  }
  CommonTextBlock_10 {
   widgetClassPath=&CommonTextBlock
   parent=HorizontalBox_5
   widget=CommonTextBlock_10
   properties {
    text="Selected!"
    textTransformPolicy=ToUpper
    font {
     bForceMonospaced=false
     bMaterialIsStencil=false
     fontMaterial=null
     fontObject=&BurbankSmall
     letterSpacing=0
     monospacedWidth=1
     size=18
     skewAmount=0
     typefaceFontName=Black
     outlineSettings {
      bApplyOutlineToDropShadows=false
      bMiteredCorners=false
      bSeparateFillAlpha=false
      outlineMaterial=null
      outlineSize=0
      outlineColor=<rgba 0 0 0 1>
     }
    }
    renderTransform {angle=0, scale=<1 1>, shear=<-8 0>, translation=<0 0>}
    shadowOffset=<0 0>
   }
   slot {class=HorizontalBoxSlot}
  }
 ]
}
